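# Unlit, two-sided translucent plasma material for the Citadel character's FX, usable on Niagara mesh particles.
# Two noise samples panning on UV0 and UV1 are summed, biased and masked by a powered UV2 vertical gradient.
# A second, tighter gradient power is added, then vertex red and particle alpha scale this mask.
# The HDR Color parameter, scaled by particle alpha, passes through MF_FX_EmissiveAdjustments for the glow.
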Material /Game/Characters/Player/Male/Medium/Bodies/M_MED_Citadel/FX/M_Citadel_Plasma
graph {
 properties {
  bUsedWithNiagaraMeshParticles=true
  blendMode=BLEND_Translucent
  pixelDepthOffsetMode=PDOM_Legacy
  shadingModel=MSM_Unlit
  twoSided=true
  usedShadingModels=Unlit
 }
 nodes [
  n1 = ParticleColor()
  MaterialFunctionCall(Emissive=Multiply(A=@n1.A, B=VectorParameter() {parameterName=Color, defaultValue=<rgba 23.875 13.75 25 1>}.RGB)) {materialFunction=&MF_FX_EmissiveAdjustments}
  n2 = Multiply(A=TextureCoordinate() {uTiling=2}, B=Constant2Vector() {g=1, r=0.5})
  n3 = Panner(Coordinate=Multiply(A=TextureCoordinate() {coordinateIndex=1, vTiling=0}, B=Constant2Vector() {g=1, r=1})) {speedX=0.5, speedY=-0.25}
  n4 = Add(
   A=TextureSample(UVs=Panner(Coordinate=@n2) {speedX=-0.5, speedY=0.1}) {texture=&T_NoisePack_01}.RGB
   B=Multiply(A=TextureSample(UVs=@n3) {texture=&T_NoisePack_01}.RGB) {constB=2}
  )
  n5 = TextureSample(UVs=TextureCoordinate() {coordinateIndex=2}) {samplerType=SAMPLERTYPE_LinearColor, texture=/Game/Global/Textures/Default/Masks_and_Gradients/T_VertGradient}
  n6 = Multiply(
   A=Subtract(A=MaterialFunctionCall(Float3=@n4) {materialFunction=&BreakOutFloat3Components}.G) {constB=0.75}
   B=Power(Base=@n5.R, Exp=ScalarParameter() {defaultValue=20, parameterName=Power1}) {constExponent=20}
  )
  n7 = Power(Base=@n5.R, Exp=ScalarParameter() {defaultValue=25, parameterName=Power2}) {constExponent=25}
  Multiply(A=@n1.A, B=Multiply(A=Add(A=@n6, B=@n7), B=VertexColor().R))
 ]
}
