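# Translucent unlit plasma detail material for Niagara mesh particles, part of the Citadel character body FX.
# A color parameter feeds an emissive adjustment function.
# A panning normal map distorts UVs for two scrolling noise textures, shifted by a dynamic-parameter random offset.
# Their combined noise drives MF_Erosion, with erosion amount taken from the maximum of a vertical UV gradient and inverted axis-aligned fresnel, alongside view-angle and depth fades.
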Material /Game/Characters/Player/Male/Medium/Bodies/M_MED_Citadel/FX/M_Citadel_Plasma_Details
graph {
 properties {
  bUsedWithNiagaraMeshParticles=true
  blendMode=BLEND_Translucent
  pixelDepthOffsetMode=PDOM_Legacy
  shadingModel=MSM_Unlit
  usedShadingModels=Unlit
 }
 nodes [
  MaterialFunctionCall(Emissive=VectorParameter() {parameterName=Color, defaultValue=<rgba 47.75 27.5 50 0>}.RGB) {materialFunction=&MF_FX_EmissiveAdjustments}
  n1 = ScalarParameter() {defaultValue=5, parameterName=Depth Fade Distance}
  Multiply(B=Power(Base=DepthFade(FadeDistance=@n1), Exp=ScalarParameter() {defaultValue=0.5, parameterName=Depth Fade Exponent}))
  n2 = TextureCoordinate() {uTiling=2}
  n3 = Constant2Vector() {g=0.125, r=0.125}
  n4 = TextureSample(UVs=Panner(Coordinate=Multiply(A=@n2, B=@n3)) {speedX=-0.1, speedY=0.1}) {samplerType=SAMPLERTYPE_Normal, texture=/Game/Global/Textures/Default/Tileable_Patterns/1Normal}
  n5 = MaterialFunctionCall(X=@n4.R, Y=@n4.R) {materialFunction=&MakeFloat2}
  n6 = DynamicParameter() {parameterIndex=2, defaultValue=<rgba 0 1 1 1>, paramNames [RandomOffsetTexture, Param2, Param3, Param4]}
  n7 = Add(A=Add(A=TextureCoordinate(), B=Multiply(A=@n5.Result) {constB=0.1}), B=@n6.RandomOffsetTexture)
  n8 = TextureSample(UVs=Panner(Coordinate=@n7) {speedX=0.1, speedY=0.75}) {
   samplerType=SAMPLERTYPE_LinearGrayscale
   texture=&/T_Citadel_Noise_Dissolve_Handpaint
  }
  n9 = TextureSample(UVs=Panner(Coordinate=@n7) {speedX=0.1, speedY=0.75}) {texture=/Game/TimeOfDay/LightFunction/T_CloudNoise_Artemis}
  n10 = Multiply(A=@n8.R, B=Multiply(A=OneMinus(None=@n9.G)) {constB=1.2})
  n11 = ScalarParameter() {defaultValue=1.2, parameterName=Texture Exponent}
  n12 = ComponentMask(None=TextureCoordinate()) {g=true}
  n13 = Add(A=Power(Base=OneMinus(None=@n12), Exp=ScalarParameter() {defaultValue=1, parameterName=Gradient Exponent})) {constB=0.4}
  n14 = Constant3Vector() {constant=<rgba 0 0 1 0>}
  n15 = OneMinus(None=MaterialFunctionCall(
    Axis=Transform(None=@n14) {transformSourceType=TRANSFORMSOURCE_Instance}
    Exponent=ScalarParameter() {defaultValue=0.4, parameterName=Fresnel Exponent}
   ) {materialFunction=&AxisAlignedFresnel}.Fresnel)
  Multiply(
   A=MaterialFunctionCall("FadeSensitivity (optional)"=Constant() {r=2}) {materialFunction=&FadeBasedOnViewAngle}.Result
   B=MaterialFunctionCall!(
    Gradient=Power(Base=@n10, Exp=@n11)
    Erosion Amount=Max(A=@n13, B=@n15)
    Softness=ScalarParameter() {defaultValue=0.5, parameterName=Erosion Softness}
   ) {materialFunction=/Game/Effects/Fort_Effects/MaterialFunctions/MF_Erosion}.Increased Range
  )
 ]
}
